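# Physics asset for the TatToon large backpack cosmetic, giving its pack and dangling attachments collision bodies and joints.
# Boxes cover the root and pack, with small capsules and spheres on the hooks, cloak pieces, cap and zippers.
# Constraint chains branch from the pack: the left hook carries a three-piece cloak, the right a cap, and two zippers attach directly.
# Pack-to-root and both base-to-pack joints rotate freely, while the others carry mostly narrow angular limits.
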PhysicsAsset /BRCosmetics/Accessories/FORT_Backpacks/Backpack_M_LRG_TatToon/Meshes/M_LRG_TatToon_Pack_Physics
bodies {
 Dyn_cap [{shapeType=Capsule, length=0.5, radius=4, shapeName=Dyn_cap_capsule, center=<9.3405 0 -0.2747>}]
 base_l [{shapeType=Capsule, length=0, radius=1, shapeName=base_l_capsule}]
 base_r [{shapeType=Capsule, length=0, radius=1, shapeName=base_r_capsule}]
 cloak_base [{shapeType=Capsule, length=0, radius=1, shapeName=cloak_base_capsule}]
 cloak_f [{shapeType=Capsule, length=0, radius=1, shapeName=cloak_f_capsule}]
 cloak_l [{shapeType=Capsule, length=1.5, radius=2.5, shapeName=cloak_l_capsule, center=<5.4162 -0.7864 -0.2662>}]
 cloak_r [{shapeType=Capsule, length=1.5, radius=2, shapeName=cloak_r_capsule, center=<6.8853 0 0>}]
 hook_l [{shapeType=Sphere, radius=0.505, shapeName=hook_l_sphere, center=<0.966 0 -1.3929>}]
 hook_r [{shapeType=Sphere, radius=2.2489, shapeName=hook_r_sphere, center=<1.9214 0.0837 -0.0281>}]
 pack [{shapeType=Box, extentX=20.5, extentY=19.5, extentZ=41.5, shapeName=pack_box, center=<1.9481 -6.7825 -10.7851>}]
 root [{shapeType=Box, extentX=10.01, extentY=57.01, extentZ=46.01, shapeName=root_box, center=<0 4.6534 -8.6681>, rotation=<rot 0 -90 0>}]
 zipper_l [{shapeType=Capsule, length=0, radius=1, shapeName=zipper_l_capsule, center=<1.7288 0 0>}]
 zipper_r [{shapeType=Capsule, length=0, radius=1, shapeName=zipper_r_capsule, center=<1.7288 0 0>}]
}
constraints [
 {
  bone1Name=base_l
  bone2Name=pack
  swing1LimitDegrees=3
  swing1Motion=Free
  swing2LimitDegrees=10
  swing2Motion=Free
  twistLimitDegrees=2
  twistMotion=Free
 }
 {bone1Name=cloak_base, bone2Name=hook_l, swing1LimitDegrees=10, swing2LimitDegrees=20, twistLimitDegrees=2}
 {bone1Name=cloak_f, bone2Name=cloak_base, swing1LimitDegrees=10, swing2LimitDegrees=20, twistLimitDegrees=2}
 {bone1Name=cloak_l, bone2Name=cloak_base, swing1LimitDegrees=10, swing2LimitDegrees=30, twistLimitDegrees=2}
 {bone1Name=cloak_r, bone2Name=cloak_base, swing1LimitDegrees=10, swing2LimitDegrees=20, twistLimitDegrees=2}
 {
  bone1Name=base_r
  bone2Name=pack
  swing1LimitDegrees=20
  swing1Motion=Free
  swing2LimitDegrees=5
  swing2Motion=Free
  twistLimitDegrees=2
  twistMotion=Free
 }
 {bone1Name=Dyn_cap, bone2Name=hook_r, swing1LimitDegrees=0, swing2LimitDegrees=60, twistLimitDegrees=2}
 {bone1Name=zipper_l, bone2Name=pack, swing1LimitDegrees=10, swing2LimitDegrees=3, twistLimitDegrees=2}
 {bone1Name=zipper_r, bone2Name=pack, swing1LimitDegrees=10, swing2LimitDegrees=3, twistLimitDegrees=2}
 {bone1Name=hook_r, bone2Name=base_r, swing1LimitDegrees=30, swing2LimitDegrees=20, twistLimitDegrees=2}
 {bone1Name=hook_l, bone2Name=base_l, swing1LimitDegrees=5, swing2LimitDegrees=20, twistLimitDegrees=2}
 {
  bone1Name=pack
  bone2Name=root
  swing1LimitDegrees=45
  swing1Motion=Free
  swing2LimitDegrees=45
  swing2Motion=Free
  twistLimitDegrees=45
  twistMotion=Free
 }
]
