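# Animation Blueprint for the TatToon backpack cosmetic, adding secondary physics motion to the pack and its Dyn_cap bone.
# An offset input pose is cached, then feeds two RigidBody and spring AnimDynamics chains differing in RigidBody alpha, weaker in front end.
# The chosen chain gets an additive cap rotation.
# Each update switches sim-space settings between skydive and ground, and computes gravity override and spine rotation, which the AnimGraph leaves unused.
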AnimBlueprint /BRCosmetics/Accessories/FORT_Backpacks/Backpack_M_LRG_TatToon/Meshes/M_LRG_TatToon_Pack_AnimBP
target_skeleton=&/M_LRG_TatToon_Pack_Skeleton
graphs [
 AnimGraph {nodes [
   n1 = RigidBody!(
    ComponentPose=Local To Component(LocalPose=Use cached pose 'pre sim'())
    Alpha=0.3
    OverrideWorldGravity:Vector="0,0,0"
    ExternalForce
    SimSpaceSettings:Sim Space Settings Structure=Get Sim Space Settings() -> (Sim Space Settings:Sim Space Settings Structure)
   ) -> (Pose) {
    bClampLinearTranslationLimitToRefPose=true
    bOverrideWorldGravity=true
    overridePhysicsAsset=&/M_LRG_TatToon_Pack_Physics
    simulationSpace=BaseBoneSpace
    alphaBoolBlend {blendOption=Linear}
    baseBoneRef {boneName=pack}
    componentAppliedLinearAccClamp=<1000 1000 1000>
    simSpaceSettings {
     maxAngularAcceleration=150
     maxAngularVelocity=50
     maxLinearAcceleration=300
     maxLinearVelocity=200
     worldAlpha=1
     externalLinearDragV=<0.7 0.7 0.7>
    }
   }
   n2 = Anim Dynamics - Bone: Dyn_cap!(ComponentPose=@n1, bAlphaBoolEnabled:bool=True, GravityOverride:Vector="0,0,0", ExternalForce="0,0,0") -> (Pose) {
    alphaInputType=Bool
    angularDampingOverride=0.9
    angularSpringConstant=2000
    bAngularSpring=true
    bOverrideAngularDamping=true
    bUseGravityOverride=true
    bUsePlanarLimit=false
    simulationSpace=BoneRelative
    alphaBoolBlend {blendOption=Linear}
    boundBone {boneName=Dyn_cap}
    componentAppliedLinearAccClamp=<2000 2000 2000>
    componentLinearAccScale=<0.4 0.4 0.4>
    componentLinearVelScale=<0.4 0.4 0.4>
    gravityOverride {z=-980}
    physicsBodyDefinitions [{
      boundBone {boneName=Dyn_cap}
      boxExtents=<30 30 10>
      constraintSetup {angularLimitsMax=<0 0 0>, angularLimitsMin=<0 0 -60>, angularTarget=<1 0 0>}
      localJointOffset=<7 0 0>
     }]
    relativeSpaceBone {boneName=pack}
   }
   n3 = Component To Local(ComponentPose=@n2)
   n4 = RigidBody(
    ComponentPose=Local To Component(LocalPose=Use cached pose 'pre sim'())
    Alpha=0.8
    OverrideWorldGravity="0,0,0"
    SimSpaceSettings=Get Sim Space Settings()
   ) {
    bClampLinearTranslationLimitToRefPose=true
    bOverrideWorldGravity=true
    overridePhysicsAsset=&M_LRG_TatToon_Pack_Physics
    simulationSpace=BaseBoneSpace
    alphaBoolBlend {blendOption=Linear}
    baseBoneRef {boneName=pack}
    componentAppliedLinearAccClamp=<1000 1000 1000>
    simSpaceSettings {
     maxAngularAcceleration=150
     maxAngularVelocity=50
     maxLinearAcceleration=300
     maxLinearVelocity=200
     worldAlpha=1
     externalLinearDragV=<0.7 0.7 0.7>
    }
   }
   n5 = Anim Dynamics - Bone: Dyn_cap(ComponentPose=@n4, bAlphaBoolEnabled=True, GravityOverride="0,0,0", ExternalForce="0,0,0") {
    alphaInputType=Bool
    angularDampingOverride=0.9
    angularSpringConstant=2000
    bAngularSpring=true
    bOverrideAngularDamping=true
    bUseGravityOverride=true
    bUsePlanarLimit=false
    simulationSpace=BoneRelative
    alphaBoolBlend {blendOption=Linear}
    boundBone {boneName=Dyn_cap}
    componentAppliedLinearAccClamp=<2000 2000 2000>
    componentLinearAccScale=<0.4 0.4 0.4>
    componentLinearVelScale=<0.4 0.4 0.4>
    gravityOverride {z=-980}
    physicsBodyDefinitions [{
      boundBone {boneName=Dyn_cap}
      boxExtents=<30 30 10>
      constraintSetup {angularLimitsMax=<0 0 0>, angularLimitsMin=<0 0 -60>, angularTarget=<1 0 0>}
      localJointOffset=<7 0 0>
     }]
    relativeSpaceBone {boneName=pack}
   }
   n6 = Component To Local(ComponentPose=@n5)
   n7 = Get bIsInFrontEnd()
   n8 = "Transform (Modify) Bone - Bone: Dyn_cap"!(
    ComponentPose=Local To Component(LocalPose=Blend Poses by bool(BlendPose_0=@n3, BlendPose_1=@n6, bActiveValue=@n7) {blendType=Linear})
    bAlphaBoolEnabled:bool=True
    Translation="0,0,0"
    Rotation="13,0,0"
    Scale="1,1,1"
   ) -> (Pose) {
    alphaInputType=Bool
    rotationMode=BMM_Additive
    rotationSpace=BCS_BoneSpace
    translationMode=BMM_Additive
    translationSpace=BCS_BoneSpace
    alphaBoolBlend {blendOption=Cubic}
    boneToModify {boneName=Dyn_cap}
    rotation {pitch=13}
   }
   Output Pose(Result=Component To Local(ComponentPose=@n8)) {layerGroup=None}
   n9 = "Transform (Modify) Bone - Bone: root"(ComponentPose=Local To Component(LocalPose=Input Pose()), Translation="0,-0.8,0", Rotation="0,0,0") {
    rotationMode=BMM_Additive
    rotationSpace=BCS_BoneSpace
    translationMode=BMM_Additive
    translationSpace=BCS_BoneSpace
    alphaBoolBlend {blendOption=Cubic}
    boneToModify {boneName=root}
    translation {y=-0.8}
   }
   n10 = "Transform (Modify) Bone - Bone: root"(ComponentPose=@n9, Translation="0,-1,0", Rotation="0,0,50") {
    rotationMode=BMM_Additive
    rotationSpace=BCS_BoneSpace
    translationMode=BMM_Additive
    translationSpace=BCS_BoneSpace
    alphaBoolBlend {blendOption=Cubic}
    boneToModify {boneName=root}
    rotation {roll=50}
    translation {y=-1}
   }
   Save cached pose 'pre sim'(Pose=Component To Local(ComponentPose=@n10)) {cacheName=pre sim}
  ]}
 EventGraph {nodes [
   Event BlueprintUpdateAnimation()
   > n1 = Sequence!(execute) -> (then_0, then_1, then_2:Exec)
   > n2 = Gravity Override State Function DEPRECATED(CharacterPartAnimInstance=Self-Reference(), GameTimeInSeconds=GetGameTimeInSeconds(), GravityOverrideParams=Get Gravity Override Params())
   > Set Gravity Override(Gravity Override=@n2)
   n3 = GetOwnerSkeletalMeshComponent()
   Set Spine rotation(
    execute=@n1.then_1
    Spine rotation:double=MapRangeClamped(
     Value=GetDeltaTransformFromRefPose!(self=@n3, BoneName=spine_05, BaseName=pelvis) -> (
      ReturnValue_Location:Vector
      ReturnValue_Rotation_Roll:float
      ReturnValue_Rotation_Pitch:float
      ReturnValue_Rotation_Yaw:float
      ReturnValue_Scale:Vector
     ).ReturnValue_Rotation_Yaw
     InRangeA=10
     InRangeB=-90
    )
   ) -> (Output_Get:double)
   > n4 = Branch(Condition=Get bIsSkyDiving())
   > Set Sim Space Settings(Sim Space Settings:Sim Space Settings Structure=Get Sim Space Settings Skydive() -> (Sim Space Settings Skydive:Sim Space Settings Structure)) -> (Output_Get:Sim Space Settings Structure)
   Set Sim Space Settings(execute=@n4.else, Sim Space Settings=Get Sim Space Settings Ground() -> (Sim Space Settings Ground:Sim Space Settings Structure))
   TryGetPawnOwner()
  ]}
]
variables [
 Gravity Override {type=Vector}
 Gravity Override Params {type=GravityOverrideParamsStruct, default {skydive_FloatingDown {gravityOverride=<0 -600 1000>}}}
 NewVar {type=boolean}
 Sim Space A {type=AnimPhysSimSpaceType}
 Sim Space Settings {
  type=SimSpaceSettings
  default {
   dampingAlpha=1
   maxAngularAcceleration=150
   maxAngularVelocity=50
   maxLinearAcceleration=300
   maxLinearVelocity=200
   velocityScaleZ=1
   worldAlpha=1
   externalLinearDragV=<0.7 0.7 0.7>
  }
 }
 Sim Space Settings Ground {
  type=SimSpaceSettings
  default {
   dampingAlpha=1
   maxAngularAcceleration=150
   maxAngularVelocity=50
   maxLinearAcceleration=1200
   maxLinearVelocity=100
   velocityScaleZ=1
   worldAlpha=1
   externalLinearDragV=<0.7 0.7 0.7>
   externalLinearVelocity=<0.5 0.5 0.5>
  }
 }
 Sim Space Settings Skydive {
  type=SimSpaceSettings
  default {
   dampingAlpha=1
   maxAngularAcceleration=150
   maxAngularVelocity=50
   maxLinearAcceleration=300
   maxLinearVelocity=200
   velocityScaleZ=1
   worldAlpha=1
   externalLinearDragV=<0.7 0.7 0.7>
  }
 }
 Spine rotation {type=number}
]
